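# Conversation database for the SquibblySquid Deep 04 NPC, covering its greeting and top-level conversation branches.
# The entry point leads to one Speech node that selects contextual lines by first-ever meeting, per-match conversation count, or specific player outfits.
# Curve-table priorities are set for first-meeting and outfit reactions, while the count-gated lines have none.
# The greeting branches through link nodes to shared quest, basic service and random service conversations, plus this NPC's own Dialogue entry, defined elsewhere.
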ConversationDatabase /NPCLibrary/NPCs/SquibblySquid_Deep/SquibblySquid_Deep_04/NPCLibrary_Conversation_SquibblySquid_Deep_04
entry_points [{tag=Conversation.Entry.SquibblySquidDeep.04, destinations [#1]}]
nodes [
 ConversationEntryPointNode_7 {
  id=#1
  class=ConversationEntryPointNode
  outputs [#2]
  properties {entryTag {tagName=Conversation.Entry.SquibblySquidDeep.04}}
 }
 Speech_C_4 {
  id=#2
  class=Speech_C
  outputs [#3, #4, #5, #6]
  properties {
   nodeName=""
   generalConfig {
    defaultMessage=Random greeting
    contextualMessages [
     {
      message=Soooo... you must be surprised to see me here...
      contextRequirements [{requirement {class=FortControllerRequirement_IsFirstEverConversationWithNPC}}]
      priority {curve {rowName=Tandem.Speech.Priority.FirstEverInteraction}}
     }
     {
      message=What can I say? I'm team Foundation.
      contextRequirements [{requirement {
         class=FortControllerRequirement_CheckNPCMatchConversationCount
         properties {nPCUniqueIdTag {tagName=AISpawnerData.Type.Tandem.SquibblySquidDeep.04}}
        }}]
      priority {curve {curveTable=null, rowName=None}}
     }
     {
      message="You can't trust everything you hear, you know. "
      contextRequirements [{requirement {
         class=FortControllerRequirement_CheckNPCMatchConversationCount
         properties {nPCUniqueIdTag {tagName=AISpawnerData.Type.Tandem.SquibblySquidDeep.04}}
        }}]
      priority {curve {curveTable=null, rowName=None}}
     }
     {
      message=You hungry?
      contextRequirements [{requirement {
         class=FortControllerRequirement_CheckNPCMatchConversationCount
         properties {nPCUniqueIdTag {tagName=AISpawnerData.Type.Tandem.SquibblySquidDeep.04}}
        }}]
      priority {curve {curveTable=null, rowName=None}}
     }
     {
      message=I'm telling you. There's no malice in his heart.
      contextRequirements [{requirement {class=FortControllerRequirement_HasCID, properties {allowedCIds [/BRCosmetics/Athena/Items/Cosmetics/Characters/Character_SquibblySquid]}}}]
      priority {curve {rowName=Tandem.Speech.Priority.OutfitReaction}}
     }
     {
      message="Hiya, boss."
      contextRequirements [{requirement {
         class=FortControllerRequirement_HasCID
         properties {allowedCIds [&CID_315_Athena_Commando_M_TeriyakiFish]}
        }}]
      priority {curve {rowName=Tandem.Speech.Priority.OutfitReaction}}
     }
    ]
   }
  }
 }
 ConversationLinkNode_3 {
  id=#6
  class=ConversationLinkNode
  properties {remoteEntryTag {tagName=Conversation.Entry.SquibblySquidDeep.04.Dialogue}}
 }
 ConversationLinkNode_2 {id=#5, class=ConversationLinkNode, properties {remoteEntryTag {tagName=Conversation.Entry.RandomServices}}}
 ConversationLinkNode_0 {id=#4, class=ConversationLinkNode, properties {remoteEntryTag {tagName=Conversation.Entry.BasicServices}}}
 ConversationLinkNode_1 {id=#3, class=ConversationLinkNode, properties {remoteEntryTag {tagName=Conversation.Entry.NPCLibrary.Quests}}}
]
