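# Conversation database for the "Ask about..." dialogue of NPC SquibblySquid Deep 04, a squid who cooks at Fishstick's restaurant.
# The tagged entry point leads to a hub speech that branches into three topics (The Foundation, Misunderstandings, Fishstick) plus a Back node.
# Each topic is a linear chain of Speech nodes carrying the character's lore and jokes.
# All three chains converge on one shared ReturnToStart node, shown as a BACK choice.
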ConversationDatabase /NPCLibrary/NPCs/SquibblySquid_Deep/SquibblySquid_Deep_04/NPCLibrary_Conversation_SquibblySquid_Deep_04_Dialogue
entry_points [{tag=Conversation.Entry.SquibblySquidDeep.04.Dialogue, destinations [#1]}]
nodes [
 ConversationEntryPointNode_7 {
  id=#1
  class=ConversationEntryPointNode
  outputs [#2]
  properties {entryTag {tagName=Conversation.Entry.SquibblySquidDeep.04.Dialogue}}
 }
 Speech_C_0 {
  id=#2
  class=Speech_C
  outputs [#3, #4, #5, #6]
  properties {nodeName="", softTaskIcon=&T_UI_NPC_question, generalConfig {defaultMessage=I'm sure you have questions.}}
  sub_nodes [ConversationChoiceNode_5 {class=ConversationChoiceNode, properties {defaultChoiceDisplayText=Ask about...}}]
 }
 Speech_C_5 {
  id=#6
  class=Speech_C
  outputs [#7]
  properties {
   nodeName=""
   softTaskIcon=&T_UI_NPC_question
   generalConfig {defaultMessage="All of my people think he's evil. "}
  }
  sub_nodes [ConversationChoiceNode_0 {class=ConversationChoiceNode, properties {defaultChoiceDisplayText=Fishstick}}]
 }
 Speech_C_6 {
  id=#7
  class=Speech_C
  outputs [#8]
  properties {nodeName="", softTaskIcon=&T_UI_NPC_question, generalConfig {defaultMessage="So, I investigated."}}
 }
 Speech_C_7 {
  id=#8
  class=Speech_C
  outputs [#9]
  properties {
   nodeName=""
   softTaskIcon=&T_UI_NPC_question
   generalConfig {defaultMessage="Turns out, not evil. Just, uh... a few fish short of a school."}
  }
 }
 Speech_C_11 {
  id=#9
  class=Speech_C
  outputs [#10]
  properties {
   nodeName=""
   softTaskIcon=&T_UI_NPC_question
   generalConfig {defaultMessage="And his restaurant? VERY profitable. And in need of a chef. So here I am!"}
  }
 }
 ReturnToStart_C_23 {
  id=#10
  class=ReturnToStart_C
  properties {softTaskIcon=&T_UI_NPC_Back}
  sub_nodes [ConversationChoiceNode_32 {class=ConversationChoiceNode, properties {defaultChoiceDisplayText=BACK}}]
 }
 Speech_C_1 {
  id=#5
  class=Speech_C
  outputs [#11]
  properties {
   nodeName=""
   softTaskIcon=&T_UI_NPC_question
   generalConfig {defaultMessage="When somebody does something bad, my first instinct isn't to hunt them for sport."}
  }
  sub_nodes [ConversationChoiceNode_1 {class=ConversationChoiceNode, properties {defaultChoiceDisplayText=Misunderstandings}}]
 }
 Speech_C_2 {
  id=#11
  class=Speech_C
  outputs [#12]
  properties {
   nodeName=""
   softTaskIcon=&T_UI_NPC_question
   generalConfig {defaultMessage="I pick up my shell-phone and ask them \"so, what was that?\""}
  }
 }
 Speech_C_3 {
  id=#12
  class=Speech_C
  outputs [#13]
  properties {
   nodeName=""
   softTaskIcon=&T_UI_NPC_question
   generalConfig {defaultMessage="From there, we can hash it out. Swim out of the weeds. And move forward."}
  }
 }
 Speech_C_4 {
  id=#13
  class=Speech_C
  outputs [#10]
  properties {
   nodeName=""
   softTaskIcon=&T_UI_NPC_question
   generalConfig {defaultMessage="It's not popular with the Squibblies, but I'm no good at holding grudges. Makes my tentacles ache. "}
  }
 }
 Speech_C_8 {
  id=#4
  class=Speech_C
  outputs [#14]
  properties {
   nodeName=""
   softTaskIcon=&T_UI_NPC_question
   generalConfig {defaultMessage=There's a platter named after him on the menu.}
  }
  sub_nodes [ConversationChoiceNode_2 {class=ConversationChoiceNode, properties {defaultChoiceDisplayText=The Foundation}}]
 }
 Speech_C_9 {
  id=#14
  class=Speech_C
  outputs [#15]
  properties {
   nodeName=""
   softTaskIcon=&T_UI_NPC_question
   generalConfig {defaultMessage="The Foundational Feast - 1100 grams of lean protein, served in a helmet-shaped chalice with a side of punch."}
  }
 }
 Speech_C_10 {
  id=#15
  class=Speech_C
  outputs [#10]
  properties {nodeName="", softTaskIcon=&T_UI_NPC_question, generalConfig {defaultMessage=Take that as you will.}}
 }
 Back_C_0 {
  id=#3
  class=Back_C
  sub_nodes [ConversationChoiceNode_7 {class=ConversationChoiceNode, properties {defaultChoiceDisplayText=BACK}}]
 }
]
